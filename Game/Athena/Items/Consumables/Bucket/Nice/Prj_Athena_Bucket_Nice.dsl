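; Projectile Blueprint for the Nice Bucket consumable, adding trail and in-air audio cleanup on explosion.
; On explosion it runs the parent logic, sets TurnOffTrail and draws a debug sphere at its location.
; The TurnOffTrail RepNotify deactivates the trail and fades out the in-air loop when true, or reactivates the trail when false.
; BeginPlay, overlap and Tick are disabled stubs.
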(event EventOnExploded
  (params (HitActors :type Array%20of%20Actor%20Object%20References) (HitResults :type Array%20of%20Hit%20Result%20Structures))
  (|CallParentFunction HitActors HitResults)
  (Variables|Default|SetTurnOffTrail true)
  (Rendering|Debug|DrawDebugSphere (Transformation|GetActorLocation) 1e+02 12 "(R=0.111639,G=0.065169,B=0.500000,A=1.000000)" 4.0 5.0))

(event EventBeginPlay
  (|CallParentFunction) ; This node is disabled and will not be called.\nDrag off pins to build functionality.
)

(event Collision|EventActorBeginOverlap
  (params (OtherActor :type Actor%20Object%20Reference))
  (|CallParentFunction OtherActor) ; This node is disabled and will not be called.\nDrag off pins to build functionality.
)

(event EventTick
  (params (DeltaSeconds :type Float%20%28single-precision%29))
  (|CallParentFunction DeltaSeconds) ; This node is disabled and will not be called.\nDrag off pins to build functionality.
)

(fn OnRep_TurnOffTrail
  (bind _trail (Variables|Default|GetTrail))
  (if (Variables|Default|GetTurnOffTrail)
    (Components|Activation|Deactivate _trail)
    (Audio|Components|Audio|FadeOut (Variables|Default|GetInAir_Loop) 0.15 0.0 "Linear" :!class /Script/Engine.AudioComponent)
    (else
      (Components|Activation|Activate _trail true))))

(fn ConstructionScript
  (|CallParentFunction))
